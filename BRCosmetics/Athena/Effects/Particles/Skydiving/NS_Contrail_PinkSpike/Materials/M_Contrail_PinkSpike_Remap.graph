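# Translucent unlit Niagara sprite material for the PinkSpike skydiving contrail, coloring a star texture through gradient remapping.
# The star texture's red channel indexes a curve atlas along U, with the Variant parameter selecting V.
# The remapped color is multiplied by particle color and passed through emissive adjustments.
# Star alpha is multiplied by particle alpha and a near-camera fade raised to the fourth power for a sharper falloff.
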Material /BRCosmetics/Athena/Effects/Particles/Skydiving/NS_Contrail_PinkSpike/Materials/M_Contrail_PinkSpike_Remap
graph {
 properties {bUsedWithNiagaraSprites=true, blendMode=BLEND_Translucent, pixelDepthOffsetMode=PDOM_Legacy, shadingModel=MSM_Unlit, usedShadingModels=Unlit}
 nodes [
  n1 = ParticleColor()
  n2 = TextureSample() {texture=/BRCosmetics/Athena/Effects/Particles/Skydiving/NS_Contrail_PinkSpike/Textures/T_Contrail_PinkSpike_Stars}
  n3 = TextureSampleParameter2D(UVs=AppendVector(A=@n2.R, B=ScalarParameter() {defaultValue=0.1, parameterName=Variant})) {
   parameterName=Curve
   samplerType=SAMPLERTYPE_LinearColor
   texture=/BRCosmetics/Athena/Effects/Particles/Skydiving/NS_Contrail_PinkSpike/Curves/Curve_Atlas_PinkSpike_StarColor
  }
  MaterialFunctionCall(Emissive=Multiply(A=@n1.RGB, B=@n3.RGB)) {materialFunction=&MF_FX_EmissiveAdjustments}
  n4 = MaterialFunctionCall(Distance=ScalarParameter() {defaultValue=40, parameterName=Near Camera Fade}) {materialFunction=&Near_Camera_Fade}
  n5 = Multiply(A=@n4.Result, B=@n4.Result)
  Multiply(A=Multiply(A=@n1.A, B=@n2.A), B=Multiply(A=@n5, B=@n5))
 ]
}
