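# Translucent material for the PinkSpike skydiving contrail, usable on Niagara ribbons and sprites.
# Color blends particle color toward a dynamic-parameter grey along U, with an adjustable exponent.
# A normal map dotted with sky atmosphere light direction shades it, and two offset sphere masks form a dimmed, crescent-like rim.
# Alpha combines particle alpha, a circular mask and near-camera fade raised to the fourth power.
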Material /BRCosmetics/Athena/Effects/Particles/Skydiving/NS_Contrail_PinkSpike/Materials/M_Contrail_PinkSpike_Ribbon
graph {
 properties {
  bUsedWithNiagaraRibbons=true
  bUsedWithNiagaraSprites=true
  blendMode=BLEND_Translucent
  pixelDepthOffsetMode=PDOM_Legacy
  usedShadingModels=Default Lit
 }
 nodes [
  n1 = ParticleColor()
  n2 = DynamicParameter() {defaultValue=<rgba 0 0.7973 2 1.5>, paramNames [R, G, B, "Lerp  Adjustment"]}
  n3 = TextureCoordinate()
  n4 = LinearInterpolate(
   A=@n1.RGB
   B=MaterialFunctionCall(X=@n2.R, Y=@n2.R, Z=@n2.R) {materialFunction=&MakeFloat3}.Result
   Alpha=ComponentMask(None=Power(Base=@n3, Exp=@n2."Lerp  Adjustment")) {r=true}
  )
  n5 = TextureSampleParameter2D() {
   parameterName=NM
   samplerType=SAMPLERTYPE_Normal
   texture=/BRCosmetics/Athena/Effects/Particles/Skydiving/NS_Contrail_PinkSpike/Textures/T_Contrail_PinkSpike_Ribbon_NM
  }
  n6 = DotProduct(A=SkyAtmosphereLightDirection(), B=Transform(None=@n5.RGB))
  n7 = Saturate(None=Add(A=Multiply(A=@n6) {constB=0.5}) {constB=0.6})
  n8 = Constant() {r=0.5}
  n9 = SphereMask(A=TextureCoordinate(), B=@n8, Radius=Constant() {r=0.48})
  n10 = TextureCoordinate() {uTiling=0.86, vTiling=1.1}
  n11 = AppendVector(A=Constant() {r=0.065}, B=Constant() {r=-0.05})
  n12 = Subtract(A=@n9, B=SphereMask(A=Add(A=@n10, B=@n11), B=@n8, Radius=Constant() {r=0.4}))
  Add(A=Multiply(A=Multiply(A=@n4, B=@n7), B=OneMinus(None=@n12)), B=Multiply(A=@n12, B=Multiply(A=@n4) {constB=0.15}))
  n13 = MaterialFunctionCall(Distance=ScalarParameter() {defaultValue=40, parameterName=Near Camera Fade}) {materialFunction=&Near_Camera_Fade}
  n14 = Multiply(A=@n13.Result, B=@n13.Result)
  Multiply(A=Multiply(A=@n1.A, B=@n9), B=Multiply(A=@n14, B=@n14))
 ]
}
