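# Stylized Subsurface material for the PinkSpike skydiving contrail's Niagara particles, forming animated, bending spiky cloud shapes.
# World position offset combines Voronoi height displacement, vertex-color spikes scaled by dynamic parameters and an optional position-based bend.
# Emissive color layers anime-style shadow, GI, AO and rim lighting, then adjusts for time of day.
# An optional variant multiplies in a dot pattern and curve-atlas gradient mapped by height from a particle-supplied bone position.
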Material /BRCosmetics/Athena/Effects/Particles/Skydiving/NS_Contrail_PinkSpike/Materials/M_Contrail_PinkSpike_Stamina_Custom
graph {
 properties {
  allowTranslucentCustomDepthWrites=true
  bEnableResponsiveAA=true
  bUsedWithNiagaraMeshParticles=true
  bUsedWithNiagaraSprites=true
  pixelDepthOffsetMode=PDOM_Legacy
  shadingModel=MSM_Subsurface
  usedShadingModels=Subsurface
 }
 nodes [
  ActorPositionWS()
  ActorPositionWS()
  n1 = NamedRerouteUsage()
  n2 = Add(
   A=Multiply(A=Transform(None=@n1), B=ScalarParameter() {parameterName=WPO_Main_NormalStrength})
   B=Multiply(A=NamedRerouteUsage(), B=NamedRerouteUsage())
  )
  n3 = DynamicParameter() {defaultValue=<rgba 1 0 1 1>, paramNames [ScaleX, Spike Dir Ctrl, Top Spike, Bottom Spike]}
  n4 = DynamicParameter() {parameterIndex=2, defaultValue=<rgba 1 1 100 0.01>, paramNames [WPO X, WPO Y, WPO Z, Bend WPO Intensity]}
  n5 = MaterialFunctionCall(X=@n4.WPO X, Y=@n4.WPO X, Z=@n4.WPO X) {materialFunction=&MakeFloat3}
  n6 = ParticlePositionWS()
  n7 = WorldPosition()
  n8 = ActorPositionWS()
  n9 = Divide(
   A=ComponentMask(None=Subtract(A=@n7.XYZ, B=@n8)) {b=true}
   B=ScalarParameter() {defaultValue=150, group=For PinkSpike, parameterName=ActorPos Z Scale}
  )
  n10 = Subtract(A=Power(Base=Saturate(None=@n9), Exp=ScalarParameter() {defaultValue=4, group=WPO, parameterName=WPOBendPower})) {constB=0.1}
  n11 = Add(A=@n10, B=@n10)
  n12 = Multiply(A=Clamp(None=Subtract(A=@n5.Result, B=@n6)) {maxDefault=300, minDefault=-300}, B=Multiply(A=@n4.Bend WPO Intensity, B=Saturate(None=@n11)))
  Add(
   A=Add(A=Multiply(A=@n2, B=@n3.ScaleX), B=NamedRerouteUsage())
   B=StaticSwitchParameter(True=Abs(None=@n12), False=Constant()) {parameterName=Use WP for WPO Flopiness}
  )
  Constant()
  Constant() {r=1}
  Constant() {r=1}
  n13 = LinearInterpolate(A=NamedRerouteUsage(), B=Constant3Vector() {constant=<rgba 0 0 0 1>})
  n14 = Multiply(A=@n13, B=NamedRerouteUsage())
  n15 = Desaturation(
   None=MaterialFunctionCall(Hue Shift Percentage=Constant(), Texture=StaticSwitchParameter(True=@n14, False=@n13) {parameterName=Use Variant}) {materialFunction=&HueShift}.Result
   Fraction=MaterialFunctionCall(Value_Morning=Constant() {r=0.1}, Value_Day=Constant(), Value_Evening=Constant() {r=0.1}, Value_Night=Constant() {r=0.05}) {materialFunction=&Scalar_Time_of_Day}.Result
  )
  n16 = MaterialFunctionCall(Value_Morning=Constant() {r=0.8}, Value_Day=Constant() {r=1}, Value_Evening=Constant() {r=0.8}, Value_Night=Constant() {r=0.9}) {materialFunction=&Scalar_Time_of_Day}
  MaterialFunctionCall(Emissive=Multiply(A=Multiply(A=@n15, B=@n16.Result), B=ParticleColor().RGB)) {materialFunction=&MF_FX_EmissiveAdjustments}
  MaterialFunctionCall(Normal=NamedRerouteUsage(), Flatness=ScalarParameter() {defaultValue=-4, group=00.Base, parameterName=Outline_Normalflat}) {materialFunction=&FlattenNormal}
  n17 = ParticleRandom()
  n18 = Time()
  NamedRerouteDeclaration(None=Multiply(A=Add(A=@n17, B=@n18), B=ScalarParameter() {defaultValue=1, group=00.Base, parameterName=Base_TimeWPO_Scale})) {name=Base_WPOTime, nodeColor=<rgba 1 0.5783 0 1>}
  n19 = MaterialFunctionCall(Float2=TextureCoordinate()) {materialFunction=&BreakOutFloat2Components}
  n20 = ScalarParameter() {defaultValue=1, parameterName=WPO_Main_VPower}
  NamedRerouteDeclaration(None=MaterialFunctionCall(X=@n19.R, Y=Power(Base=@n19.G, Exp=@n20)) {materialFunction=&MakeFloat2}.Result) {name=Base_WPOUV, nodeColor=<rgba 0.0312 1 0 1>}
  n21 = Constant()
  n22 = VertexColor()
  n23 = DynamicParameter() {defaultValue=<rgba 1 0 1 1>, paramNames [ScaleX, Spike Dir Ctrl, Top Spike, Bottom Spike]}
  n24 = Multiply(A=@n22.G, B=@n23.Bottom Spike)
  n25 = Multiply(A=@n24, B=@n23.Spike Dir Ctrl)
  n26 = Multiply(A=@n22.B, B=@n23.Top Spike)
  n27 = Multiply(A=@n23.Top Spike, B=@n26)
  NamedRerouteDeclaration(None=Add(
    A=MaterialFunctionCall(X=@n21, Y=@n25, Z=@n24) {materialFunction=&MakeFloat3}.Result
    B=MaterialFunctionCall(X=@n21, Y=@n27, Z=@n26) {materialFunction=&MakeFloat3}.Result
   )) {name=Vertex Spike, nodeColor=<rgba 1 0 0.3516 1>}
  n28 = MaterialFunctionCall(Float2=TextureCoordinate()) {materialFunction=&BreakOutFloat2Components}
  n29 = Sine(None=Multiply(A=@n28.G) {constB=0.5})
  n30 = Power(Base=@n29, Exp=ScalarParameter() {defaultValue=2, group=20.Mask, parameterName=Mask_ShapeTex_B_UpPower})
  n31 = Power(Base=@n29, Exp=ScalarParameter() {defaultValue=5, group=20.Mask, parameterName=Mask_ShapeTex_B_DownPower})
  n32 = Saturate(None=Multiply(A=Subtract(A=@n28.G) {constB=0.475}) {constB=20})
  NamedRerouteDeclaration(None=Multiply(A=VertexColor().R, B=LinearInterpolate(A=@n30, B=@n31, Alpha=@n32))) {name=Mask_ShapeTex_B, nodeColor=<rgba 1 0.2405 0 1>}
  n33 = TextureObjectParameter() {
   group=10.Cloud_Main
   parameterName=Cloud_Main_HeightTex
   texture=/BRCosmetics/Athena/Effects/Particles/Skydiving/NS_Contrail_PinkSpike/Textures/T_PinkSpike_Voronoi2D_Height
  }
  n34 = Add(A=Multiply(A=@n17) {constB=100}, B=VectorParameter() {group=10.Cloud_Main, parameterName=Cloud_Main_UVOffset}.RGB)
  n35 = VectorParameter() {group=10.Cloud_Main, parameterName=Cloud_Main_UVControl, defaultValue=<rgba 1 1 0 0>}
  n36 = NamedRerouteUsage()
  n37 = NamedRerouteUsage()
  n38 = Add(
   A=Power(
    Base=MaterialFunctionCall(Texture=@n33, AdditionalUVOffset=@n34, UVControl_RGB=@n35.RGB, UVControl_A=@n35.RGB, "[Override]TimeInput"=@n36, "[Override]UVInput"=@n37) {materialFunction=&FXCommon_TextureSample}.R
    Exp=ScalarParameter() {defaultValue=1, group=10.Cloud_Main, parameterName=Cloud_Main_WPOPower}
   )
   B=ScalarParameter() {defaultValue=-0.5, group=10.Cloud_Main, parameterName=Cloud_Main_WPOStrengthOffset}
  ) {constB=-0.5}
  n39 = ScalarParameter() {defaultValue=100, group=10.Cloud_Main, parameterName=Cloud_Main_WPOStrength}
  NamedRerouteDeclaration(None=Multiply(A=Multiply(A=@n38, B=@n39), B=VertexNormalWS())) {name=Cloud_Main_WPO, nodeColor=<rgba 1 0 0.7266 1>}
  n40 = TextureObjectParameter() {
   group=10.Cloud_Main
   parameterName=Cloud_Main_NormalTex
   samplerType=SAMPLERTYPE_Normal
   texture=/BRCosmetics/Athena/Effects/Particles/Skydiving/NS_Contrail_PinkSpike/Textures/T_PinkSpike_Voronoi2D_NM
  }
  NamedRerouteDeclaration(None=LinearInterpolate(
    A=Constant3Vector() {constant=<rgba 0 0 1 0>}
    B=MaterialFunctionCall(Texture=@n40, AdditionalUVOffset=@n34, UVControl_RGB=@n35.RGB, UVControl_A=@n35.RGB, "[Override]TimeInput"=@n36, "[Override]UVInput"=@n37) {materialFunction=&FXCommon_TextureSample}.RGB
    Alpha=NamedRerouteUsage()
   )) {name=Cloud_Main_Normal, nodeColor=<rgba 0 0.5427 1 1>}
  n41 = VectorParameter() {group=29.Color, parameterName=Color_OutShadow, defaultValue=<rgba 1 0 0.6663 1>}
  n42 = VectorParameter() {group=29.Color, parameterName=Color_OutShadow_Rim, defaultValue=<rgba 0 1 0.674 1>}
  n43 = NamedRerouteUsage()
  n44 = OneMinus(None=DotProduct(A=Transform(None=@n43), B=CameraVectorWS()))
  n45 = Saturate(None=Power(Base=Saturate(None=@n44), Exp=ScalarParameter() {defaultValue=5, group=39.RimLight, parameterName=RimLight_POW}))
  n46 = NamedRerouteUsage()
  n47 = ScalarParameter() {group=30.LightingAnimeShadow, parameterName=Lighting_AnimeShadow_NormalFlat}
  n48 = DotProduct(
   A=NamedRerouteUsage()
   B=Transform(None=MaterialFunctionCall(Normal=@n46, Flatness=@n47) {materialFunction=&FlattenNormal}.Result)
  )
  n49 = ScalarParameter() {group=30.LightingAnimeShadow, parameterName=Lighting_AnimeShadow_NDS_Offset}
  n50 = Saturate(None=Multiply(A=Add(A=@n48, B=@n49), B=ScalarParameter() {defaultValue=1, group=30.LightingAnimeShadow, parameterName=Lighting_AnimeShadow_NDS_Mul}))
  n51 = ScalarParameter() {defaultValue=1, group=30.LightingAnimeShadow, parameterName=Lighting_AnimeShadow_NDS_Power}
  n52 = VectorParameter() {group=30.LightingAnimeShadow, parameterName=Lighting_AnimeShadow_NDS_Remap, defaultValue=<rgba 0 1 0 1>}
  n53 = MaterialFunctionCall(Input=Saturate(None=Power(Base=@n50, Exp=@n51)), Input Low=Constant(), Input High=Constant() {r=1}, Target Low=@n52.R, Target High=@n52.R) {materialFunction=&RemapValueRange}
  n54 = NamedRerouteUsage()
  n55 = Transform(None=MaterialFunctionCall(Normal=NamedRerouteUsage(), Flatness=ScalarParameter() {group=32.LightingGI, parameterName=Lighting_GI_NormalFlat}) {materialFunction=&FlattenNormal}.Result)
  n56 = Multiply(
   A=Add(A=DotProduct(A=@n54, B=@n55), B=ScalarParameter() {group=32.LightingGI, parameterName=Lighting_GI_NDS_Offset})
   B=ScalarParameter() {defaultValue=1, group=32.LightingGI, parameterName=Lighting_GI_NDS_Mul}
  )
  n57 = Saturate(None=Power(Base=Saturate(None=@n56), Exp=ScalarParameter() {defaultValue=1, group=32.LightingGI, parameterName=Lighting_GI_NDS_Power}))
  n58 = Constant()
  n59 = Constant() {r=1}
  n60 = VectorParameter() {group=32.LightingGI, parameterName=Lighting_GI_NDS_Remap, defaultValue=<rgba 0 1 0 1>}
  n61 = LinearInterpolate(
   A=VectorParameter() {group=29.Color, parameterName=Color_GI, defaultValue=<rgba 0 0.047 1 1>}.RGB
   B=LinearInterpolate(
    A=VectorParameter() {group=29.Color, parameterName=Color_InShadow, defaultValue=<rgba 0.9722 1 0 1>}.RGB
    B=LinearInterpolate(A=@n41.RGB, B=@n42.RGB, Alpha=@n45)
    Alpha=Saturate(None=@n53.Result)
   )
   Alpha=Saturate(None=MaterialFunctionCall(Input=@n57, Input Low=@n58, Input High=@n59, Target Low=@n60.R, Target High=@n60.R) {materialFunction=&RemapValueRange}.Result)
  )
  n62 = Desaturation(
   None=MaterialFunctionCall(Hue Shift Percentage=ScalarParameter() {group=31.LightingB, parameterName=Lighting_AO_HueShift}, Texture=@n61) {materialFunction=&HueShift}.Result
   Fraction=ScalarParameter() {group=31.LightingB, parameterName=Lighting_AO_Desaturation}
  )
  n63 = MaterialFunctionCall(Normal=NamedRerouteUsage(), Flatness=ScalarParameter() {group=31.LightingAO, parameterName=Lighting_AO_NormalFlat}) {materialFunction=&FlattenNormal}
  n64 = Add(
   A=DotProduct(A=NamedRerouteUsage(), B=Transform(None=@n63.Result))
   B=ScalarParameter() {group=31.LightingAO, parameterName=Lighting_AO_NDS_Offset}
  )
  n65 = ScalarParameter() {defaultValue=1, group=31.LightingAO, parameterName=Lighting_AO_NDS_Mul}
  n66 = Power(
   Base=Saturate(None=Multiply(A=@n64, B=@n65))
   Exp=ScalarParameter() {defaultValue=1, group=31.LightingAO, parameterName=Lighting_AO_NDS_Power}
  )
  n67 = VectorParameter() {group=31.LightingAO, parameterName=Lighting_AO_NDS_Remap, defaultValue=<rgba 0 1 0 1>}
  n68 = Saturate(None=MaterialFunctionCall(Input=Saturate(None=@n66), Input Low=Constant(), Input High=Constant() {r=1}, Target Low=@n67.R, Target High=@n67.R) {materialFunction=&RemapValueRange}.Result)
  NamedRerouteDeclaration(None=Multiply(A=LinearInterpolate(A=@n62, B=@n61, Alpha=@n68), B=@n68)) {name=LightingOutEmissive, nodeColor=<rgba 0 1 0.2969 1>}
  n69 = DynamicParameter() {parameterIndex=3, defaultValue=<rgba 0 1 1 0>, paramNames [Frontend_ShadingPosOfs, Frontend_PatOfsMul, Frontend_VariScaleMul, none]}
  n70 = WorldPosition()
  n71 = DynamicParameter() {parameterIndex=3, defaultValue=<rgba 0 1 1 0>, paramNames [Frontend_ShadingPosOfs, Frontend_PatOfsMul, Frontend_VariScaleMul, none]}
  n72 = ObjectPositionWS()
  n73 = Constant3Vector() {constant=<rgba 0 0 165 0>}
  n74 = DynamicParameter() {parameterIndex=1, defaultValue=<rgba 0 0 0 1>, paramNames [Bone Pos X, Bone Pos Y, Bone Pos Z, Variant V Ofs Scale]}
  n75 = Add(
   A=Add(A=@n71.Frontend_ShadingPosOfs, B=Subtract(A=@n72, B=@n73))
   B=MaterialFunctionCall(X=@n74.Bone Pos X, Y=@n74.Bone Pos X, Z=@n74.Bone Pos X) {materialFunction=&MakeFloat3}.Result
  )
  n76 = Add(
   A=Multiply(A=@n69.Frontend_PatOfsMul, B=ScalarParameter() {defaultValue=11, group=For PinkSpike, parameterName=Variant Tex V Ofs})
   B=ComponentMask(None=Subtract(A=@n70.XYZ, B=@n75)) {b=true}
  )
  n77 = ScalarParameter() {defaultValue=0.006, group=For PinkSpike, parameterName=Variant Tex V Scale}
  n78 = Power(
   Base=Saturate(None=Multiply(A=@n76, B=@n77))
   Exp=ScalarParameter() {defaultValue=4, group=For PinkSpike, parameterName=Variant Mask Under Power}
  ) {constExponent=5}
  n79 = OneMinus(None=@n78)
  n80 = ScalarParameter() {defaultValue=1, group=For PinkSpike, parameterName=Variant UV U Scale}
  n81 = ScalarParameter() {defaultValue=1, group=For PinkSpike, parameterName=Variant UV V Scale}
  n82 = Add(A=Multiply(A=TextureCoordinate(), B=AppendVector(A=@n80, B=@n81)) {constB=3}, B=ParticleRandom())
  n83 = Multiply(A=Time(), B=AppendVector(A=Constant(), B=ScalarParameter() {defaultValue=0.7, group=For PinkSpike, parameterName="Variant UV V  Pan"}))
  n84 = TextureSampleParameter2D(UVs=StaticSwitchParameter(True=Add(A=@n82, B=@n83), False=@n82) {parameterName=Use Variant Tex Scroll}) {
   group=For PinkSpike
   parameterName=Variant Surface Tex
   samplerType=SAMPLERTYPE_LinearColor
   texture=/BRCosmetics/Athena/Effects/Particles/Skydiving/NS_Contrail_PinkSpike/Textures/T_Contrail_PinkSpike_variant2_dot
  }
  n85 = Multiply(
   A=StaticSwitchParameter(True=Multiply(A=@n79, B=@n84.RGB), False=@n84.RGB) {parameterName=Use Variant Tex Mask}
   B=ScalarParameter() {defaultValue=1, group=For PinkSpike, parameterName=Variant Tex Color Scale}
  )
  n86 = OneMinus(None=@n84.A)
  n87 = ScalarParameter() {defaultValue=5, group=For PinkSpike, parameterName=Variant Textute V Noise Intensity}
  n88 = TextureCoordinate()
  n89 = ScalarParameter() {defaultValue=4, group=For PinkSpike, parameterName=Variant Textute V Noise Scale}
  n90 = Power(
   Base=TextureSampleParameter2D(UVs=Multiply(A=@n88, B=@n89)) {group=For PinkSpike, parameterName=Variant Color Noise Tex, texture=/Game/Effects/Fort_Effects/Textures/Fire/T_FireNoise_02}.RGB
   Exp=ScalarParameter() {defaultValue=1, group=For PinkSpike, parameterName=Variant Textute V Noise Power}
  )
  n91 = DynamicParameter() {parameterIndex=3, defaultValue=<rgba 0 1 1 0>, paramNames [Frontend_ShadingPosOfs, Frontend_PatOfsMul, Frontend_VariScaleMul, none]}
  n92 = Subtract(A=ObjectPositionWS(), B=Constant3Vector() {constant=<rgba 0 0 165 0>})
  n93 = DynamicParameter() {parameterIndex=1, defaultValue=<rgba 0 0 0 1>, paramNames [Bone Pos X, Bone Pos Y, Bone Pos Z, Variant V Ofs Scale]}
  n94 = Subtract(
   A=Add(A=WorldPosition().XYZ, B=Multiply(A=@n87, B=@n90))
   B=Add(
    A=Add(A=@n91.Frontend_ShadingPosOfs, B=@n92)
    B=MaterialFunctionCall(X=@n93.Bone Pos X, Y=@n93.Bone Pos X, Z=@n93.Bone Pos X) {materialFunction=&MakeFloat3}.Result
   )
  )
  n95 = ScalarParameter() {defaultValue=11, group=For PinkSpike, parameterName=Variant V Ofs}
  n96 = DynamicParameter() {parameterIndex=3, defaultValue=<rgba 0 1 1 0>, paramNames [Frontend_ShadingPosOfs, Frontend_PatOfsMul, Frontend_VariScaleMul, none]}
  n97 = Multiply(
   A=Add(A=ComponentMask(None=@n94) {b=true}, B=Multiply(A=@n95, B=@n93.Variant V Ofs Scale))
   B=Multiply(A=ScalarParameter() {defaultValue=0.006, group=For PinkSpike, parameterName=Variant V Scale}, B=@n96.Frontend_VariScaleMul)
  )
  n98 = TextureSample(UVs=AppendVector(A=Saturate(None=@n97), B=ScalarParameter() {defaultValue=0.1, group=For PinkSpike, parameterName=Variant Tex V})) {
   samplerType=SAMPLERTYPE_LinearColor
   texture=/BRCosmetics/Athena/Effects/Particles/Skydiving/NS_Contrail_PinkSpike/Curves/Curve_Atlas_Contrail_PinkSpike
  }
  n99 = Multiply(A=Saturate(None=Add(A=@n78, B=@n86)), B=@n98.RGB)
  NamedRerouteDeclaration(None=Add(A=Multiply(A=@n85, B=@n84.A), B=Saturate(None=@n99))) {name=Variant Color, nodeColor=<rgba 0.4062 0 1 1>}
  n100 = VectorParameter() {parameterName=FELightDir}
  NamedRerouteDeclaration(None=LinearInterpolate(A=SkyAtmosphereLightDirection(), B=@n100.RGB, Alpha=@n100.RGB)) {name=LightDir, nodeColor=<rgba 0 0.2031 1 1>}
  VectorParameter() {group=00.Base, parameterName=SSSMask, defaultValue=<rgba 1 0 0 1>}
 ]
}
