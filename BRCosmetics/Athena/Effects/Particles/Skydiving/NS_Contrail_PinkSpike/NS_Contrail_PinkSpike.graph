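# PinkSpike skydiving contrail Niagara system, with separate tuning for gameplay and front-end preview.
# Mesh smoke and star sprites spawn at two bones or sockets on the user skeletal mesh, alongside ribbons and GPU speedlines; Stars also emits location events.
# Diving attributes drive spawn rates, speeds and shapes, with smoke writing sine-swayed custom position and scale.
# User.Variant_Float selects ribbon colors, material parameters and renderer visibility tags.
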NiagaraSystem /BRCosmetics/Athena/Effects/Particles/Skydiving/NS_Contrail_PinkSpike/NS_Contrail_PinkSpike
dependencies {
 usedDataInterfaces [
  &NiagaraDataInterfaceCurve
  &NiagaraDataInterfaceArrayColor
  &NiagaraDataInterfaceArrayFloat
  &NiagaraDataInterfaceColorCurve
  &NiagaraDataInterfaceVector2DCurve
  &NiagaraDataInterfacePlatformSet
  &NiagaraDataInterfaceCurlNoise
 ]
 usedDynamicInputs [
  &MakeCustomVectorFromBool
  &TransformVector
  &SetBoolByFloatComparison
 ]
 usedModules [
  &SystemState
  /Game/Effects/Niagara/Modules/Skydiving/FN_DivingAttributes/v5/FN_DivingAttributes
  &self:SystemSpawnScript.SystemScriptSource.SystemScriptGraph.NiagaraNodeAssignment_0.SetVariables_#1
  &self:SystemSpawnScript.SystemScriptSource.SystemScriptGraph.NiagaraNodeAssignment_1.SetVariables_#2
  &FN_Integrator
  &self:SystemSpawnScript.SystemScriptSource.SystemScriptGraph.NiagaraNodeAssignment_2.SetVariables_#3
  &self:SystemSpawnScript.SystemScriptSource.SystemScriptGraph.NiagaraNodeAssignment_5.SetVariables_#4
  &self:SystemSpawnScript.SystemScriptSource.SystemScriptGraph.NiagaraNodeAssignment_3.SetVariables_#5
 ]
}
emitters {
 Ribbon_Follow {
  emitterUpdateScript [&EmitterState()]
  particleSpawnScript [
   &InitializeRibbon(
    Color=&SelectLinearColorFromArray(
     Array Sampling Mode=::NewEnumerator1
     Direct Array Index=&MakeIntFromFloat(Float=$User.Variant_Float)
     Color Selection Array=DataInterface() {propertyValues {ColorData [<rgba 0 0.7973 2 1>, <rgba 2 1.4389 0 1>, <rgba 0.0327 0.0262 0.3507 1>]}}
     Random Seed=?
    )
    Lifetime=0.8
    Ribbon Width=1
    UsePositionOffset=false
    Write Material Random=true
   )
   &CameraOffset(Camera Offset Amount=50)
   &AddVelocity(Velocity=<0 0 18>)
   &DynamicMaterialParameters(
    Index 0 Param 1=&SelectFloatFromArray(
     Direct Array Index=&MakeIntFromFloat(Float=$User.Variant_Float)
     Float Selection Array=DataInterface() {propertyValues {FloatData [0,2,1]}}
    )
    Index 0 Param 2=&SelectFloatFromArray(
     Direct Array Index=&MakeIntFromFloat(Float=$User.Variant_Float)
     Float Selection Array=DataInterface() {propertyValues {FloatData [0.7973,1.4,1.5]}}
    )
    Index 0 Param 3=&SelectFloatFromArray(
     Direct Array Index=&MakeIntFromFloat(Float=$User.Variant_Float)
     Float Selection Array=DataInterface() {propertyValues {FloatData [2,0,2]}}
    )
    Index 0 Param 4=1.5
    Module.Param0WriteEnabled=true
    Module.Param1WriteEnabled=true
    Module.Param2WriteEnabled=true
    Module.Param3WriteEnabled=true
   )
  ]
  particleUpdateScript [
   &ParticleState()
   &CurlNoiseForce(
    Noise Frequency=&FloatFromCurve(FloatCurve=DataInterface() {propertyValues {Curve {keys [<at 0.5 1>,<at 1 0.3>]}}}, Scale Curve=20)
    Noise Quality / Cost=::NewEnumerator0
    Noise Strength=&FloatFromCurve(
     FloatCurve=DataInterface() {propertyValues {Curve {keys [{tangentMode=RCTM_User, time=0.5, value=1}, {tangentMode=RCTM_User, time=1, value=2}]}}}
     Scale Curve=30
    )
    Pan Noise=false
   ) {enabled=false}
   &SolveForcesAndVelocity(Clamp Velocity=false, Limit Acceleration=false)
   &ScaleColor(
    Linear Color Curve=DataInterface() {propertyValues {
      AlphaCurve {keys [<at 0.0082 0>,<at 0.3205 1>,<at 0.8224 1>,<at 1 0>]}
      BlueCurve {keys [<at 0 1>,<at 1 1>]}
      GreenCurve {keys [<at 0 1>,<at 1 1>]}
      RedCurve {keys [<at 0 1>,<at 1 1>]}
     }}
    Scale Mode=::NewEnumerator2
   )
   &ScaleColor(
    Scale Alpha=&FloatFromCurve(
     FloatCurve=DataInterface() {propertyValues {Curve {keys [{interpMode=RCIM_Cubic, tangentMode=RCTM_User, time=0.3, value=1}, <at 1 0>]}}}
     CurveIndex=$Emitter.NormalizedLoopAge
    )
    ScaleA=true
    ScaleRGB=false
   ) {enabled=false, module=ScaleColor001}
   &ScaleRibbonWidth(Ribbon Width Scale=&FloatFromCurve(FloatCurve=DataInterface() {propertyValues {Curve {keys [
         {interpMode=RCIM_Cubic, tangentMode=RCTM_User, time=0, value=0}
         {interpMode=RCIM_Cubic, tangentMode=RCTM_User, time=0.5, value=4}
         <at 0.9 0>
        ]}}}))
  ]
  rendererClasses [&NiagaraRibbonRendererProperties]
  renderers [{rendererClass=&NiagaraRibbonRendererProperties}]
 }
 Smoke_Large {
  emitterUpdateScript [
   &EmitterState(Life Cycle Mode=::NewEnumerator1, Loop Duration=2, UseLoopDelay=false)
   &SpawnRate(SpawnRate=7, Use Spawn Probability=false)
   /SetVariables(
    Emitter.EdgeThickness=0.5
    Emitter.FresnelThickEXP:NiagaraFloat=&self:PlatFromSelect(
     Default:NiagaraFloat=50
     IsTarget:NiagaraFloat=1.1
     TargetPlatform:NiagaraDataInterfacePlatformSet=DataInterface() {propertyValues {Platforms {
        qualityLevelMask=-32
        deviceProfileStates [
         {profileName=IOS, qualityLevelMask=4294967295, setQualityLevelMask=4294967295}
         {profileName=Android, qualityLevelMask=4294967295, setQualityLevelMask=4294967295}
         {profileName="Android_Mali_G72 ", qualityLevelMask=4294967295, setQualityLevelMask=4294967295}
         {profileName=Android_Galaxy_Z_Fold2_Adreno, qualityLevelMask=4294967295, setQualityLevelMask=4294967295}
         {profileName=Android_Galaxy_Z_Fold2_Adreno_LRU, qualityLevelMask=4294967295, setQualityLevelMask=4294967295}
         {profileName=Android_Galaxy_Z_Fold2_Adreno_Vulkan, qualityLevelMask=4294967295, setQualityLevelMask=4294967295}
         {profileName=Android_High, qualityLevelMask=4294967295, setQualityLevelMask=4294967295}
         {profileName=IOS_Epic, qualityLevelMask=4294967295, setQualityLevelMask=4294967295}
        ]
       }}}
    )
   ) {isSetParameters=true, module=SetVariables_#6}
  ]
  particleSpawnScript [
   &FN_SetSpawnGroup(Number of Spawn Groups=2)
   &InitializeParticle(
    Color=/Niagara/DynamicInputs/TypeConversions/MakeLinearColorFromFloat(Value:NiagaraFloat=&FN_FrontEndFloatModifier(Gameplay=1, Front End=2))
    Lifetime=&Lerp_Float(
     A=&RandomRangeFloat(Minimum=0.7, Maximum=0.8)
     B=&RandomRangeFloat(Minimum=0.65, Maximum=0.78)
     Alpha=$System.RDive
    )
    Mesh Scale Mode=::NewEnumerator1
    Mesh Uniform Scale Max=0.4
    Mesh Uniform Scale Min=0.37
    Sprite Rotation Mode=::NewEnumerator2
    UsePositionOffset=false
    Write Material Random=true
   )
   &CylinderLocation(Cylinder Height=0.5, Cylinder Radius=0.1, MaskSpawnGroup=false, Override Seed=false, UseOwnerScale=true)
   &AddVelocity(
    Cone Angle=3
    Cone Axis=&MaskVectorBySpawnGroup(Masked Vector=<0.1 0 1>, Passthrough Vector=<-0.1 0 1>)
    Module.UseConeFalloff=false
    Module.UseVelDistribution=false
    Override Seed=false
    Velocity Mode=::NewEnumerator2
    Velocity Speed=&Lerp_Float(A=130, B=200, Alpha=$System.Dive)
   )
   &InitialMeshOrientation(
    Mesh Axis To Orient=&MaskVectorBySpawnGroup(Masked Vector=<1 0 0>, Passthrough Vector=<-1 0 0>)
    Mesh Orientation Mode=::NewEnumerator2
    Orientation Vector=&FN_FrontEndVector3Modifier(
     Gameplay=&MaskVectorBySpawnGroup(Masked Vector=<-0.5 0 -0.5>, Passthrough Vector=<0.5 0 0.5>)
     Front End=&MaskVectorBySpawnGroup(Masked Vector=<0.7 0 1>, Passthrough Vector=<-1 0 0>)
    )
    Rotation=&FN_FrontEndVector3Modifier(Gameplay=<0 0 0>, Front End=<0 0.5 0>)
    Rotation Coordinate Space=::NewEnumerator2
    Use Rotation Vector=true
   )
   /SetVariables(Particles.VisibilityTag=&MakeIntFromFloat(Float=$User.Variant_Float)) {isSetParameters=true}
  ]
  particleUpdateScript [
   &ParticleState()
   &CurlNoiseForce(
    Noise Frequency=500
    Noise Strength=&FloatFromCurve(FloatCurve=DataInterface() {propertyValues {Curve {keys [<at 0 0.5>,<at 0.25 0.5>,<at 0.4 2>,<at 0.6 0>]}}}, Scale Curve=350)
    Pan Noise=true
   ) {enabled=false}
   &UpdateMeshOrientation(
    Rotation Coordinate Space=::NewEnumerator0
    Rotation Rate=30
    Rotation Vector=&RandomVector(Vector Scale=&FloatFromCurve(FloatCurve=DataInterface() {propertyValues {Curve {keys [{interpMode=RCIM_Cubic, tangentMode=RCTM_User, time=0.2, value=0}, {interpMode=RCIM_Cubic, tangentMode=RCTM_User, time=5, value=1}]}}}))
   ) {enabled=false}
   &UpdateMeshOrientation(
    Facing Direction=<1 0 1>
    Facing Interpolation=false
    Orientation Method=::NewEnumerator1
    Override Quat=false
    TransformFacing=true
    TransformUp=false
    Up Direction:Vector3f=&FN_FrontEndVector3Modifier(Gameplay=<0 -0.05 1>, Front End=&Lerp_Vector(A=<0 -0.05 1>, B=<0 0 1>, Alpha=$System.Dive))
   ) {module=UpdateMeshOrientation001}
   &ScaleMeshSize(Scale Factor=&VectorFromFloat(Value=&FloatFromCurve(FloatCurve=DataInterface() {propertyValues {Curve {keys [
          {interpMode=RCIM_Cubic, leaveTangent=0.913, tangentMode=RCTM_Break, time=0, value=0.05}
          {arriveTangent=0.8235, interpMode=RCIM_Cubic, tangentMode=RCTM_Break, time=0.25, value=1}
          {arriveTangent=-0.3858, interpMode=RCIM_Cubic, leaveTangent=-0.3858, tangentMode=RCTM_User, time=0.7, value=0.8}
          {arriveTangent=-0.8076, interpMode=RCIM_Cubic, leaveTangent=-0.8076, tangentMode=RCTM_User, time=1, value=0}
         ]}}})))
   &SkeletalMeshLocation(
    Bone / Socket Index=$Particles.SpawnGroup
    Bone Sampling Mode=::NewEnumerator4
    MaskSpawnGroup=false
    Orientation Sampling=::NewEnumerator1
    Position Sampling=::NewEnumerator2
    Skeletal Mesh=$User.SkeletalMesh
   )
   &SolveForcesAndVelocity(Clamp Velocity=false, Limit Acceleration=false)
   /SetVariables(Particles.NPosition:Vector3f=&Add_Vector(
     A=&Add_Vector(
      A=Linked() {linkedVariable {name=Particles.Position, type=NiagaraPosition}}
      B=&FN_FrontEndVector3Modifier(
       Gameplay=&Lerp_Vector(
        A=<0 0 40>
        B=&MaskVectorBySpawnGroup(Masked Vector=<10 0 40>, Passthrough Vector=<-15 0 40>)
        Alpha=$System.RDive
       )
       Front End=&Lerp_Vector(
        A=&MaskVectorBySpawnGroup(Masked Vector=<8 0 40>, Passthrough Vector=<-8 0 40>)
        B=&MaskVectorBySpawnGroup(Masked Vector=<7 -10 40>, Passthrough Vector=<-10 -10 40>)
        Alpha=$System.Dive
       )
      )
     )
     B=&Multiply_VectorByFloat(
      Vector=&Lerp_Vector(
       A=<0 0 0>
       B=<7 0 0>
       Alpha=&ClampFloat(Float=&Divide_Float(
         A=&MakeFloatFromVector(Vector=Linked() {linkedVariable {name=Particles.Position, type=NiagaraPosition}}, Channel=::NewEnumerator2)
         B=0
        ))
      )
      Float=&Sine(
       Normalized Angle=&Add_Float(
        A=&Add_Float(
         A=&Multiply_Float(A=$System.Age, B=1)
         B=&MakeCustomFloatFromBool(
          Bool=&SetBoolByIntComparison(A=&SpawnGroup(), B=0, Comparison Type=::NewEnumerator4)
          True Float=0.43
          False Float=0
         )
        )
        B=&Divide_Float(
         A=&MakeFloatFromVector(Vector=Linked() {linkedVariable {name=Particles.Position, type=NiagaraPosition}}, Channel=::NewEnumerator2)
         B=&MaskFloatBySpawnGroup(Masked Float=350, Passthrough Float=400)
        )
       )
       Period=1
       Scale=1
       Bias=0
      )
     )
    )) {isSetParameters=true, module=SetVariables_#7}
   /SetVariables(Particles.NPosition:Vector3f=&Add_Vector(
     A=/Niagara/DynamicInputs/Math/TransformPositionByMatrix(
      Position:Vector3f=$Particles.NPosition
      TransformationMatrix:NiagaraMatrix=/Niagara/DynamicInputs/Math/MakeTransform(
       Inverse:NiagaraBool=false
       Position:Vector3f=?
       Local Offset:Vector3f=<0 0 0>
       Scale:Vector3f=<1 1 1>
       Rotation Interface:ENiagara_RotationDefinitionApproach=::NewEnumerator2
       Axis:Vector3f=?
       Quaternion:Quat4f=?
       Rotation Matrix:NiagaraMatrix=?
       X Vector:Vector3f=?
       Yaw:NiagaraFloat=0
       Angle:NiagaraFloat=?
       Pitch:NiagaraFloat=&Lerp_Float(
        A=&MakeCustomFloatFromBool(
         Bool=&SetBoolByIntComparison(A=$Particles.SpawnGroup, B=0, Comparison Type=::NewEnumerator4)
         True Float=5
         False Float=-5
        )
        B=&MakeCustomFloatFromBool(
         Bool=&SetBoolByIntComparison(A=$Particles.SpawnGroup, B=0, Comparison Type=::NewEnumerator4)
         True Float=5.5
         False Float=-5.5
        )
        Alpha=$System.Dive
       )
       Y Vector:Vector3f=?
       Roll:NiagaraFloat=&Lerp_Float(
        A=-10
        B=&MakeCustomFloatFromBool(Bool=$User.bIsFrontEndPreview, True Float=-30, False Float=0)
        Alpha=$System.Dive
       )
       Z Vector:Vector3f=?
      )
     )
     B=$Particles.SkeletalMeshLocation.SampledPosition
    )) {isSetParameters=true, module=SetVariables_#8}
   &ScaleMeshSize(Scale Factor=&FN_FrontEndVector3Modifier(
     Gameplay=&Lerp_Vector(A=<0.7 0.7 0.7>, B=<0.2 0.2 0.8>, Alpha=$System.RDive)
     Front End=&Lerp_Vector(A=<0.7 0.7 0.7>, B=<0.2 0.2 0.5>, Alpha=$System.RDive)
    )) {module=ScaleMeshSize001}
   /SetVariables(Particles.NScale:Vector3f=&Multiply_VectorByFloat(
     Vector=&Lerp_Vector(A=<1 1 1.1>, B=<0.7 0.7 1.1>, Alpha=$System.RDive)
     Float=&FloatFromCurve(
      FloatCurve=DataInterface() {propertyValues {Curve {keys [
          {interpMode=RCIM_Cubic, leaveTangent=7.6731, tangentMode=RCTM_Break, time=0, value=0.2}
          {arriveTangent=1.526, interpMode=RCIM_Cubic, tangentMode=RCTM_Break, time=0.28, value=1.15}
          {arriveTangent=-0.3858, interpMode=RCIM_Cubic, leaveTangent=-0.3858, tangentMode=RCTM_User, time=0.75, value=1.1}
          {arriveTangent=-2.1353, interpMode=RCIM_Cubic, leaveTangent=-2.1353, tangentMode=RCTM_User, time=1, value=0.65}
         ]}}}
      Scale Curve=0.49
     )
    )) {isSetParameters=true, module=SetVariables_#9}
   &DynamicMaterialParameters(
    Index 0 Param 1=1
    Index 0 Param 2=&FN_FrontEndFloatModifier(Gameplay=0, Front End=&Lerp_Float(A=0, B=-0.9, Alpha=$System.Dive))
    Index 0 Param 3=&FloatFromCurve(
     FloatCurve=DataInterface() {propertyValues {Curve {keys [
         {interpMode=RCIM_Cubic, tangentMode=RCTM_User, time=0.65, value=0}
         {interpMode=RCIM_Cubic, tangentMode=RCTM_User, time=0.8, value=0.9}
         {arriveTangent=-4.5, interpMode=RCIM_Cubic, leaveTangent=-4.5, time=0.9, value=1}
         {interpMode=RCIM_Cubic, tangentMode=RCTM_User, time=1, value=0}
        ]}}}
     Scale Curve=10
    )
    Index 0 Param 4=&FloatFromCurve(
     FloatCurve=DataInterface() {propertyValues {Curve {keys [
         {interpMode=RCIM_Cubic, tangentMode=RCTM_User, time=0.05, value=1}
         {interpMode=RCIM_Cubic, tangentMode=RCTM_User, time=0.4, value=0}
         <at 0.8 0.125>
         <at 1 0.125>
        ]}}}
     Scale Curve=&FN_FrontEndFloatModifier(Gameplay=-10, Front End=&Lerp_Float(A=-20, B=0, Alpha=$System.Dive))
    )
    Index 1 Param 1=HlslExpression() {hlslExpression=Particles.SkeletalMeshLocation.SampledPosition.x}
    Index 1 Param 2=HlslExpression() {hlslExpression=Particles.SkeletalMeshLocation.SampledPosition.y}
    Index 1 Param 3=HlslExpression() {hlslExpression=Particles.SkeletalMeshLocation.SampledPosition.z}
    Index 1 Param 4=&FN_FrontEndFloatModifier(
     Gameplay=1
     Front End=&SelectFloatFromArray(
      Direct Array Index=&MakeIntFromFloat(Float=$User.Variant_Float)
      Float Selection Array=DataInterface() {propertyValues {FloatData [0,-1.1,-0.2]}}
     )
    )
    Index 3 Param 1=&FN_FrontEndFloatModifier(Gameplay=0, Front End=-30)
    Index 3 Param 2=&FN_FrontEndFloatModifier(Gameplay=1, Front End=2)
    Index 3 Param 3=&FN_FrontEndFloatModifier(Gameplay=1, Front End=1.4)
    Module.Param0WriteEnabled=true
    Module.Param0WriteEnabled1=true
    Module.Param0WriteEnabled3=true
    Module.Param1WriteEnabled=true
    Module.Param1WriteEnabled1=true
    Module.Param1WriteEnabled3=true
    Module.Param2WriteEnabled=false
    Module.Param2WriteEnabled1=true
    Module.Param2WriteEnabled3=true
    Module.Param3WriteEnabled=true
    Module.Param3WriteEnabled1=true
    Module.Param3WriteEnabled3=false
    Write Parameter Index 1=true
    Write Parameter Index 3=true
   )
  ]
  rendererClasses [&NiagaraMeshRendererProperties]
  renderers [
   {rendererClass=&NiagaraMeshRendererProperties, rendererIndex=0}
   {rendererClass=&NiagaraMeshRendererProperties, rendererIndex=1}
   {rendererClass=&NiagaraMeshRendererProperties, rendererIndex=2}
   {rendererClass=&NiagaraMeshRendererProperties, rendererIndex=3}
   {rendererClass=&NiagaraMeshRendererProperties, rendererIndex=4}
   {rendererClass=&NiagaraMeshRendererProperties, rendererIndex=5}
   {rendererClass=&NiagaraMeshRendererProperties, rendererIndex=6}
   {rendererClass=&NiagaraMeshRendererProperties, rendererIndex=7}
   {rendererClass=&NiagaraMeshRendererProperties, rendererIndex=8}
  ]
 }
 Smoke_Small {
  emitterUpdateScript [
   &EmitterState(Life Cycle Mode=::NewEnumerator1, Loop Delay=0.1, Loop Duration=2, UseLoopDelay=true)
   &SpawnRate(SpawnRate=&Lerp_Float(A=25, B=30, Alpha=$System.Dive), Use Spawn Probability=false)
   /SetVariables(
    Emitter.EdgeThickness=0.55
    Emitter.FresnelThickEXP:NiagaraFloat=&self:PlatFromSelect(
     Default:NiagaraFloat=50
     IsTarget:NiagaraFloat=1.1
     TargetPlatform:NiagaraDataInterfacePlatformSet=DataInterface() {propertyValues {Platforms {
        qualityLevelMask=-32
        deviceProfileStates [
         {profileName=IOS, qualityLevelMask=4294967295, setQualityLevelMask=4294967295}
         {profileName=Android, qualityLevelMask=4294967295, setQualityLevelMask=4294967295}
         {profileName="Android_Mali_G72 ", qualityLevelMask=4294967295, setQualityLevelMask=4294967295}
         {profileName=Android_Galaxy_Z_Fold2_Adreno, qualityLevelMask=4294967295, setQualityLevelMask=4294967295}
         {profileName=Android_Galaxy_Z_Fold2_Adreno_LRU, qualityLevelMask=4294967295, setQualityLevelMask=4294967295}
         {profileName=Android_Galaxy_Z_Fold2_Adreno_Vulkan, qualityLevelMask=4294967295, setQualityLevelMask=4294967295}
         {profileName=Android_High, qualityLevelMask=4294967295, setQualityLevelMask=4294967295}
         {profileName=IOS_Epic, qualityLevelMask=4294967295, setQualityLevelMask=4294967295}
        ]
       }}}
    )
   ) {isSetParameters=true, module=SetVariables_#6}
  ]
  particleSpawnScript [
   &FN_SetSpawnGroup(Number of Spawn Groups=2)
   &InitializeParticle(
    Color=&MakeLinearColorFromFloat(Value:NiagaraFloat=&FN_FrontEndFloatModifier(Gameplay=1, Front End=2))
    Lifetime=&Lerp_Float(
     A=&RandomRangeFloat(Minimum=0.6, Maximum=0.8)
     B=&RandomRangeFloat(Minimum=0.65, Maximum=0.9)
     Alpha=$System.RDive
    )
    Mesh Scale Mode=::NewEnumerator1
    Mesh Uniform Scale Max=0.35
    Mesh Uniform Scale Min=0.3
    Sprite Rotation Mode=::NewEnumerator2
    UsePositionOffset=false
    Write Material Random=true
   )
   &CylinderLocation(Cylinder Height=0.5, Cylinder Radius=0.1, MaskSpawnGroup=false, Override Seed=false, UseOwnerScale=true)
   &AddVelocity(
    Cone Angle=1
    Cone Axis=&MaskVectorBySpawnGroup(Masked Vector=<0.1 0 1>, Passthrough Vector=<-0.1 0 1>)
    Module.UseConeFalloff=false
    Module.UseVelDistribution=false
    Override Seed=false
    Velocity Mode=::NewEnumerator2
    Velocity Speed=&Lerp_Float(
     A=&RandomRangeFloat(Minimum=130, Maximum=150)
     B=&RandomRangeFloat(Minimum=220, Maximum=240)
     Alpha=$System.Dive
    )
   )
   &InitialMeshOrientation(
    Mesh Axis To Orient=&MaskVectorBySpawnGroup(Masked Vector=<1 0 0>, Passthrough Vector=<-1 0 0>)
    Mesh Orientation Mode=::NewEnumerator2
    Orientation Vector=<0 0 1>
    Rotation=&RandomRangeVector(Minimum=<0 0 0>, Maximum=<0 0 0>)
    Rotation Coordinate Space=::NewEnumerator2
    Use Rotation Vector=true
   )
   &CameraOffset(Camera Offset Amount=&Lerp_Float(A=3, B=-6, Alpha=$System.RDive))
   /SetVariables(Particles.VisibilityTag=&MakeIntFromFloat(Float=$User.Variant_Float)) {isSetParameters=true}
  ]
  particleUpdateScript [
   &ParticleState()
   &CurlNoiseForce(
    Noise Frequency=500
    Noise Strength=&FloatFromCurve(FloatCurve=DataInterface() {propertyValues {Curve {keys [<at 0 0.5>,<at 0.25 0.5>,<at 0.4 2>]}}}, Scale Curve=50)
    Pan Noise=true
   )
   &SolveForcesAndVelocity(Clamp Velocity=false, Limit Acceleration=false)
   &ScaleMeshSize(Scale Factor=&VectorFromFloat(Value=&FloatFromCurve(FloatCurve=DataInterface() {propertyValues {Curve {keys [
          {interpMode=RCIM_Cubic, leaveTangent=14.871, tangentMode=RCTM_Break, time=0, value=0.2}
          {arriveTangent=0.2908, interpMode=RCIM_Cubic, tangentMode=RCTM_Break, time=0.17, value=1}
          {arriveTangent=-0.3858, interpMode=RCIM_Cubic, leaveTangent=-0.3858, tangentMode=RCTM_User, time=0.8, value=0.85}
          {arriveTangent=-0.8076, interpMode=RCIM_Cubic, leaveTangent=-0.8076, tangentMode=RCTM_User, time=1, value=0}
         ]}}})))
   &SkeletalMeshLocation(
    Bone / Socket Index=$Particles.SpawnGroup
    Bone Sampling Mode=::NewEnumerator4
    MaskSpawnGroup=false
    Position Sampling=::NewEnumerator2
    Skeletal Mesh=$User.SkeletalMesh
   )
   /SetVariables(Particles.NPosition:Vector3f=&Add_Vector(
     A=&Add_Vector(
      A=Linked() {linkedVariable {name=Particles.Position, type=NiagaraPosition}}
      B=&FN_FrontEndVector3Modifier(
       Gameplay=&Lerp_Vector(
        A=<0 0 30>
        B=&MaskVectorBySpawnGroup(Masked Vector=<5 -6 30>, Passthrough Vector=<-8 -6 30>)
        Alpha=$System.RDive
       )
       Front End=&Lerp_Vector(
        A=&MaskVectorBySpawnGroup(Masked Vector=<0 0 30>, Passthrough Vector=<-5 0 30>)
        B=&MaskVectorBySpawnGroup(Masked Vector=<5 -6 30>, Passthrough Vector=<-8 -6 30>)
        Alpha=$System.RDive
       )
      )
     )
     B=&Multiply_VectorByFloat(
      Vector=&Lerp_Vector(
       A=<0 0 0>
       B=<7 0 0>
       Alpha=&ClampFloat(Float=&Divide_Float(
         A=&MakeFloatFromVector(Vector=Linked() {linkedVariable {name=Particles.Position, type=NiagaraPosition}}, Channel=::NewEnumerator2)
         B=60
        ))
      )
      Float=&Sine(
       Normalized Angle=&Add_Float(
        A=&Add_Float(
         A=&Multiply_Float(A=$System.Age, B=1)
         B=&MakeCustomFloatFromBool(
          Bool=&SetBoolByIntComparison(A=&SpawnGroup(), B=0, Comparison Type=::NewEnumerator4)
          True Float=0.43
          False Float=0
         )
        )
        B=&Divide_Float(
         A=&MakeFloatFromVector(Vector=Linked() {linkedVariable {name=Particles.Position, type=NiagaraPosition}}, Channel=::NewEnumerator2)
         B=&MaskFloatBySpawnGroup(Masked Float=350, Passthrough Float=400)
        )
       )
       Period=1
       Scale=1
       Bias=0
      )
     )
    )) {isSetParameters=true, module=SetVariables_#7}
   /SetVariables(Particles.NPosition:Vector3f=&Add_Vector(
     A=&TransformPositionByMatrix(
      Position:Vector3f=$Particles.NPosition
      TransformationMatrix:NiagaraMatrix=&MakeTransform(
       Inverse:NiagaraBool=false
       Position:Vector3f=?
       Local Offset:Vector3f=<0 0 0>
       Scale:Vector3f=<1 1 1>
       Rotation Interface:ENiagara_RotationDefinitionApproach=::NewEnumerator2
       Axis:Vector3f=?
       Quaternion:Quat4f=?
       Rotation Matrix:NiagaraMatrix=?
       X Vector:Vector3f=?
       Yaw:NiagaraFloat=0
       Angle:NiagaraFloat=?
       Pitch:NiagaraFloat=&Lerp_Float(
        A=&MakeCustomFloatFromBool(
         Bool=&SetBoolByIntComparison(A=$Particles.SpawnGroup, B=0, Comparison Type=::NewEnumerator4)
         True Float=5
         False Float=-5
        )
        B=&MakeCustomFloatFromBool(
         Bool=&SetBoolByIntComparison(A=$Particles.SpawnGroup, B=0, Comparison Type=::NewEnumerator4)
         True Float=5.5
         False Float=-5.5
        )
        Alpha=$System.Dive
       )
       Y Vector:Vector3f=?
       Roll:NiagaraFloat=&Lerp_Float(
        A=-10
        B=&MakeCustomFloatFromBool(Bool=$User.bIsFrontEndPreview, True Float=-30, False Float=0)
        Alpha=$System.Dive
       )
       Z Vector:Vector3f=?
      )
     )
     B=$Particles.SkeletalMeshLocation.SampledPosition
    )) {isSetParameters=true, module=SetVariables_#8}
   &ScaleMeshSize(Scale Factor=&VectorFromFloat(Value=&Lerp_Float(A=0.73, B=0.67, Alpha=$System.RDive))) {module=ScaleMeshSize001}
   /SetVariables(Particles.NScale:Vector3f=&Multiply_VectorByFloat(Vector=$Particles.Scale, Float=0.9)) {isSetParameters=true, module=SetVariables_#9}
   &UpdateMeshOrientation(
    Rotation Coordinate Space=::NewEnumerator0
    Rotation Rate=30
    Rotation Vector=&RandomVector(Vector Scale=&FloatFromCurve(FloatCurve=DataInterface() {propertyValues {Curve {keys [{interpMode=RCIM_Cubic, tangentMode=RCTM_User, time=0.2, value=0}, {interpMode=RCIM_Cubic, tangentMode=RCTM_User, time=5, value=1}]}}}))
   )
   &DynamicMaterialParameters(
    Index 0 Param 1=1
    Index 0 Param 2=&FN_FrontEndFloatModifier(Gameplay=0, Front End=&Lerp_Float(A=0, B=-0.9, Alpha=$System.Dive))
    Index 0 Param 3=&FloatFromCurve(
     FloatCurve=DataInterface() {propertyValues {Curve {keys [
         {interpMode=RCIM_Cubic, tangentMode=RCTM_User, time=0.6, value=0}
         <at 0.9 1 RCIM_Cubic>
         {interpMode=RCIM_Cubic, tangentMode=RCTM_User, time=1, value=0}
        ]}}}
     Scale Curve=&FN_FrontEndFloatModifier(Gameplay=15, Front End=0)
    )
    Index 0 Param 4=&FloatFromCurve(
     FloatCurve=DataInterface() {propertyValues {Curve {keys [{interpMode=RCIM_Cubic, tangentMode=RCTM_User, time=0.05, value=1}, {interpMode=RCIM_Cubic, tangentMode=RCTM_User, time=0.4, value=0}]}}}
     Scale Curve=&FN_FrontEndFloatModifier(Gameplay=-35, Front End=&Lerp_Float(A=-35, B=0, Alpha=$System.Dive))
    )
    Index 1 Param 1=HlslExpression() {hlslExpression=Particles.SkeletalMeshLocation.SampledPosition.x}
    Index 1 Param 2=HlslExpression() {hlslExpression=Particles.SkeletalMeshLocation.SampledPosition.y}
    Index 1 Param 3=HlslExpression() {hlslExpression=Particles.SkeletalMeshLocation.SampledPosition.z}
    Index 1 Param 4=&FN_FrontEndFloatModifier(
     Gameplay=1
     Front End=&SelectFloatFromArray(
      Direct Array Index=&MakeIntFromFloat(Float=$User.Variant_Float)
      Float Selection Array=DataInterface() {propertyValues {FloatData [0,-1.1,-0.2]}}
     )
    )
    Index 3 Param 1=&FN_FrontEndFloatModifier(Gameplay=0, Front End=-30)
    Index 3 Param 2=&FN_FrontEndFloatModifier(Gameplay=1, Front End=2)
    Index 3 Param 3=&FN_FrontEndFloatModifier(Gameplay=1, Front End=1.4)
    Module.Param0WriteEnabled=true
    Module.Param0WriteEnabled1=true
    Module.Param0WriteEnabled3=true
    Module.Param1WriteEnabled=true
    Module.Param1WriteEnabled1=true
    Module.Param1WriteEnabled3=true
    Module.Param2WriteEnabled=false
    Module.Param2WriteEnabled1=true
    Module.Param2WriteEnabled3=true
    Module.Param3WriteEnabled=true
    Module.Param3WriteEnabled1=true
    Module.Param3WriteEnabled3=false
    Write Parameter Index 1=true
    Write Parameter Index 3=true
   ) {module=DynamicMaterialParameters001}
  ]
  rendererClasses [&NiagaraMeshRendererProperties]
  renderers [
   {rendererClass=&NiagaraMeshRendererProperties, rendererIndex=0}
   {rendererClass=&NiagaraMeshRendererProperties, rendererIndex=1}
   {rendererClass=&NiagaraMeshRendererProperties, rendererIndex=2}
   {rendererClass=&NiagaraMeshRendererProperties, rendererIndex=3}
   {rendererClass=&NiagaraMeshRendererProperties, rendererIndex=4}
   {rendererClass=&NiagaraMeshRendererProperties, rendererIndex=5}
   {rendererClass=&NiagaraMeshRendererProperties, rendererIndex=6}
   {rendererClass=&NiagaraMeshRendererProperties, rendererIndex=7}
   {rendererClass=&NiagaraMeshRendererProperties, rendererIndex=8}
  ]
 }
 Speedlines {
  simTarget=GPUComputeSim
  emitterUpdateScript [
   &EmitterState(Life Cycle Mode=::NewEnumerator1, Loop Duration=1, UseLoopDelay=false)
   &SpawnRate(
    SpawnRate=&Multiply_Float(
     A=$System.FN_Integrator001.Value
     B=&FN_FrontEndFloatModifier(Gameplay=&UniformRangedFloat(Minimum=100, Maximum=200), Front End=0)
    )
    Use Spawn Probability=false
   )
  ]
  particleSpawnScript [
   /Niagara/Modules/Spawn/Initialization/InitializeParticle(
    Color=&MakeLinearColorFromVectorAndFloat(
     "Vector (RGB)"=<1 1 1>
     "Float (Alpha)"=&Multiply_Float(
      A=&Multiply_Float(
       A=&UniformRangedFloat(Minimum=0.015, Maximum=0.15)
       B=&FN_FrontEndFloatModifier(Gameplay=1, Front End=0.5)
      )
      B=&Add_Float(
       A=1
       B=&Multiply_Float(A=10, B=&Subtract_Float(A=1, B=$Engine.Emitter.SpawnCountScale))
      )
     )
    )
    Lifetime=&FN_FrontEndFloatModifier(
     Gameplay=&UniformRangedFloat(Minimum=0.15, Maximum=0.3)
     Front End=&UniformRangedFloat(Minimum=0.1, Maximum=0.2)
    )
    Module.Write Color=true
    Module.Write Lifetime=true
    Module.Write Mass=true
    Module.Write Material Random=true
    Module.Write Position=false
    Module.Write Scale=false
    Module.Write SpriteRotation=false
    Module.Write SpriteSize=true
    Position=&SimulationPosition()
    Sprite Size=&UniformRangedVector2D(Minimum=<3 4>, Maximum=<5 6>)
   )
   &CylinderLocation(
    Cylinder Radius=&UniformRangedFloat(Minimum=100, Maximum=150)
    MaskSpawnGroup=false
    Override Seed=false
    UseOwnerScale=true
   )
   &AddVelocity(Velocity=&FN_FrontEndVector3Modifier(Gameplay=<0 0 1000>, Front End=<0 0 3000>))
   &AddVelocityFromPoint(
    Module.Use Falloff Distance=false
    Velocity Strength=&Multiply_Float(
     A=&FN_FrontEndFloatModifier(Gameplay=1, Front End=2)
     B=&UniformRangedFloat(Minimum=500, Maximum=700)
    )
   )
  ]
  particleUpdateScript [
   &ParticleState()
   &SolveForcesAndVelocity(Clamp Velocity=false, Limit Acceleration=false)
   &ScaleSpriteSize(Scale Factor=&Vector2DFromCurve(
     Vector2Curve=DataInterface() {propertyValues {XCurve {keys [<at 0 0.5 RCIM_Cubic>,<at 0.5 1 RCIM_Cubic>]}, YCurve {keys [<at 0 0 RCIM_Cubic>,<at 0.5 1 RCIM_Cubic>]}}}
     ScaleCurve=&Vector2DFromFloat(Value=1)
    ))
   &ScaleSpriteSizeBySpeed(Max Scale Factor=<1 40>, Min Scale Factor=<0.01 0.05>, Velocity Threshold=1)
   &ScaleColor(
    Scale Alpha=&FloatFromCurve(
     FloatCurve=DataInterface() {propertyValues {Curve {keys [<at 0 0 RCIM_Cubic>, <at 0.2 1>, {arriveTangent=-2.221, interpMode=RCIM_Cubic, leaveTangent=-2.221, tangentMode=RCTM_User, time=1, value=0}]}}}
     Scale Curve=&FN_FrontEndFloatModifier(Gameplay=0.4, Front End=0)
    )
    ScaleA=true
    ScaleRGB=true
   )
  ]
  rendererClasses [&NiagaraSpriteRendererProperties]
  renderers [{rendererClass=&NiagaraSpriteRendererProperties}]
 }
 Stars {
  emitterUpdateScript [
   &EmitterState(
    Inactive Response=::NewEnumerator1
    Life Cycle Mode=::NewEnumerator1
    Loop Duration=1
    Recalculate Duration Each Loop=true
    UseLoopDelay=false
   )
   &SpawnBurst_Instantaneous(
    Spawn Count=&Add_Integer(A=&RandomRangeInt(Minimum=1, Maximum=2), B=HlslExpression() {hlslExpression=System.Dive *4})
    Spawn Time=&RandomRangeFloat(Minimum=0.1, Maximum=0.1)
    Use Spawn Probability=false
    UseLoopCountLimit=false
   ) {enabled=false}
   &SpawnRate(SpawnRate=&Lerp_Float(A=3.5, B=10, Alpha=$System.Dive), Use Spawn Probability=false)
  ]
  particleSpawnScript [
   &FN_SetSpawnGroup(Number of Spawn Groups=2) {module=FN_SetSpawnGroup001}
   /Niagara/Modules/Spawn/Initialization/V2/InitializeParticle(
    Lifetime=&FN_FrontEndFloatModifier(
     Gameplay=&UniformRangedFloat(Minimum=0.7, Maximum=0.9)
     Front End=&UniformRangedFloat(Minimum=0.9, Maximum=1)
    )
    Sprite Rotation Mode=::NewEnumerator2
    Sprite Size Mode=::NewEnumerator3
    Sprite UV Mode=::NewEnumerator3
    Uniform Sprite Size=&Multiply_Float(
     A=&RandomRangeFloat(Minimum=15, Maximum=17)
     B=HlslExpression() {hlslExpression="step((rand(1.0)-0.22),0)"}
    )
    UsePositionOffset=false
    Write Material Random=true
   )
   &SkeletalMeshLocation(
    Bone / Socket Index=$Particles.SpawnGroup
    Bone Sampling Mode=::NewEnumerator4
    Inherit Velocity=false
    MaskSpawnGroup=false
    Sampled Position Offset=&Lerp_Vector(
     A=&RandomRangeVector(Minimum=<0 0 -30>, Maximum=<0 0 90>)
     B=&FN_FrontEndVector3Modifier(
      Gameplay=&RandomRangeVector(Minimum=<0 0 -30>, Maximum=<0 0 90>)
      Front End=&RandomRangeVector(Minimum=<0 0 0>, Maximum=<0 0 90>)
     )
     Alpha=$System.RDive
    )
    Sampled Position Offset Coordinate Space:ENiagaraMeshTransforms=::NewEnumerator2
    Skeletal Mesh=$User.SkeletalMesh
   )
   &AddVelocityInCone(
    Cone Angle=&FN_FrontEndFloatModifier(Gameplay=48, Front End=48)
    Cone Axis=<0 0 1>
    Module.Use Velocity Falloff On Cone Axis=false
    Velocity Strength=&FN_FrontEndFloatModifier(
     Gameplay=&UniformRangedFloat(Minimum=200, Maximum=320)
     Front End=&RandomRangeFloat(Minimum=160, Maximum=250)
    )
   )
   &ApplyInitialForces()
   &AddVelocity(Velocity=&FN_FrontEndVector3Modifier(Gameplay=<0 0 71>, Front End=&RandomRangeVector(Minimum=<-50 -50 100>, Maximum=<50 50 100>))) {enabled=false}
   &CameraOffset(Camera Offset Amount=30)
   &SubUVAnimation(End Frame=1, SubUV Animation Mode:ENiagara_SubUVLookupMode=::NewEnumerator2)
   /SetVariables(Particles.VisibilityTag=&MakeIntFromFloat(Float=$User.Variant_Float)) {isSetParameters=true}
  ]
  particleUpdateScript [
   &ParticleState()
   &GravityForce(
    Coordinate Space=::Local
    Gravity=&FN_FrontEndVector3Modifier(
     Gameplay=&RandomRangeVector(Minimum=<-50 -50 -380>, Maximum=<50 50 -500>)
     Front End=&RandomRangeVector(Minimum=<-300 -300 -200>, Maximum=<300 300 -200>)
    )
   )
   &ScaleSpriteSize(Uniform Curve Sprite Scale=DataInterface() {propertyValues {Curve {keys [
        {tangentMode=RCTM_User, time=0, value=0}
        <at 0.15 1>
        {arriveTangent=-0.9021, interpMode=RCIM_Cubic, leaveTangent=-0.9021, tangentMode=RCTM_User, time=0.5, value=1}
        {arriveTangent=-4.4458, interpMode=RCIM_Cubic, leaveTangent=-4.4458, tangentMode=RCTM_User, time=1, value=0}
       ]}}})
   &SpriteRotationRate(Rotation Rate=&RandomRangeFloat(Minimum=-300, Maximum=300))
   &CurlNoiseForce(
    Curl Noise Cone Mask Axis:Vector3f=&Multiply_Vector(A=<1 1 1>, B=<1 1 1>)
    Curl Noise Cone Mask Falloff Angle=30
    Mask Curl Noise=true
    Noise Frequency=20
    Noise Quality / Cost=::NewEnumerator0
    Noise Strength=&FN_FrontEndFloatModifier(Gameplay=500, Front End=0)
    Pan Noise=true
    Pan Noise Field=<1 0.5 0>
   ) {enabled=false}
   &InheritSourceMovement(
    Applied Force Scale=<-0.15 -0.15 -0.15>
    Applied Position Scale=<0 0 -0.01>
    ApplyForce=false
    ApplyPosition=true
    ApplyPositionFalloff=false
   )
   &Drag(
    Drag=&FN_FrontEndFloatModifier(Gameplay=1, Front End=1.2)
    Rotational Drag=0.1
    Use Linear Drag=true
    Use Rotational Drag=false
   )
   &SolveForcesAndVelocity(Clamp Velocity=false, Limit Acceleration=false)
   &SolveRotationalForcesAndVelocity()
   &GenerateLocationEvent(Event Probability=1, Use Event Probability=true, UseEventDelay=false)
  ]
  rendererClasses [&NiagaraSpriteRendererProperties]
  renderers [
   {rendererClass=&NiagaraSpriteRendererProperties}
   {rendererClass=&NiagaraSpriteRendererProperties, rendererIndex=1}
   {rendererClass=&NiagaraSpriteRendererProperties, rendererIndex=2}
  ]
 }
 Stars_No_Trail {
  simTarget=GPUComputeSim
  emitterUpdateScript [
   &EmitterState(
    Inactive Response=::NewEnumerator1
    Life Cycle Mode=::NewEnumerator1
    Loop Behavior=::NewEnumerator1
    Loop Duration Mode=::NewEnumerator1
    UseLoopDelay=false
   )
   &SpawnBurst_Instantaneous(
    Spawn Count=20
    Spawn Time=&RandomRangeFloat(Minimum=0.1, Maximum=0.1)
    Use Spawn Probability=false
    UseLoopCountLimit=false
   )
   &SpawnRate(
    SpawnRate=&Lerp_Float(A=8, B=&FN_FrontEndFloatModifier(Gameplay=25, Front End=14), Alpha=$System.Dive)
    Use Spawn Probability=false
   )
  ]
  particleSpawnScript [
   &FN_SetSpawnGroup(Number of Spawn Groups=2) {module=FN_SetSpawnGroup001}
   &InitializeParticle(
    Lifetime=&FN_FrontEndFloatModifier(
     Gameplay=&UniformRangedFloat(Minimum=0.7, Maximum=0.9)
     Front End=&UniformRangedFloat(Minimum=0.9, Maximum=1)
    )
    Sprite Rotation Mode=::NewEnumerator2
    Sprite Size Mode=::NewEnumerator1
    Sprite UV Mode=::NewEnumerator3
    Uniform Sprite Size Max=7.5
    Uniform Sprite Size Min=6
    UsePositionOffset=false
    Write Material Random=true
   )
   &SkeletalMeshLocation(
    Bone / Socket Index=$Particles.SpawnGroup
    Bone Sampling Mode=::NewEnumerator4
    Inherit Velocity=false
    MaskSpawnGroup=false
    Sampled Position Offset=&Lerp_Vector(
     A=&RandomRangeVector(Minimum=<0 0 -10>, Maximum=<0 0 140>)
     B=&FN_FrontEndVector3Modifier(
      Gameplay=&RandomRangeVector(Minimum=<0 0 -10>, Maximum=<0 0 140>)
      Front End=&RandomRangeVector(Minimum=<0 0 -10>, Maximum=<0 0 140>)
     )
     Alpha=$System.Dive
    )
    Sampled Position Offset Coordinate Space:ENiagaraMeshTransforms=::NewEnumerator2
    Skeletal Mesh=$User.SkeletalMesh
   )
   &AddVelocityInCone(
    Cone Angle=&FN_FrontEndFloatModifier(Gameplay=47, Front End=47)
    Cone Axis=<0 0 1>
    Module.Use Velocity Falloff On Cone Axis=false
    Velocity Strength=&FN_FrontEndFloatModifier(
     Gameplay=&UniformRangedFloat(Minimum=200, Maximum=320)
     Front End=&RandomRangeFloat(Minimum=160, Maximum=250)
    )
   )
   &ApplyInitialForces()
   &AddVelocity(Velocity=&FN_FrontEndVector3Modifier(Gameplay=<0 0 71>, Front End=&RandomRangeVector(Minimum=<-50 -50 100>, Maximum=<50 50 100>))) {enabled=false}
   &CameraOffset(Camera Offset Amount=30)
   &SubUVAnimation(End Frame=1, SubUV Animation Mode:ENiagara_SubUVLookupMode=::NewEnumerator2)
   /SetVariables(Particles.VisibilityTag=&MakeIntFromFloat(Float=$User.Variant_Float)) {isSetParameters=true, module=SetVariables_#10}
  ]
  particleUpdateScript [
   &ParticleState()
   &GravityForce(
    Coordinate Space=::Local
    Gravity=&FN_FrontEndVector3Modifier(
     Gameplay=&RandomRangeVector(Minimum=<-50 -50 -380>, Maximum=<50 50 -500>)
     Front End=&RandomRangeVector(Minimum=<-300 -300 -200>, Maximum=<300 300 -200>)
    )
   )
   &ScaleSpriteSize(Uniform Curve Sprite Scale=DataInterface() {propertyValues {Curve {keys [
        {tangentMode=RCTM_User, time=0, value=0}
        <at 0.15 1>
        {arriveTangent=-0.9021, interpMode=RCIM_Cubic, leaveTangent=-0.9021, tangentMode=RCTM_User, time=0.5, value=1}
        {arriveTangent=-4.4458, interpMode=RCIM_Cubic, leaveTangent=-4.4458, tangentMode=RCTM_User, time=1, value=0}
       ]}}})
   &SpriteRotationRate(Rotation Rate=&RandomRangeFloat(Minimum=-500, Maximum=500))
   &CurlNoiseForce(
    Noise Frequency=20
    Noise Quality / Cost=::NewEnumerator0
    Noise Strength=&FN_FrontEndFloatModifier(Gameplay=700, Front End=50)
    Pan Noise=true
    Pan Noise Field=<1 0.5 0>
   )
   &InheritSourceMovement(
    Applied Force Scale=<-0.15 -0.15 -0.15>
    Applied Position Scale=<0 0 -0.015>
    ApplyForce=false
    ApplyPosition=true
    ApplyPositionFalloff=false
   )
   &Drag(
    Drag=&FN_FrontEndFloatModifier(Gameplay=1, Front End=1.2)
    Rotational Drag=0.1
    Use Linear Drag=true
    Use Rotational Drag=false
   )
   &SolveForcesAndVelocity(Clamp Velocity=false, Limit Acceleration=false)
   &SolveRotationalForcesAndVelocity()
  ]
  rendererClasses [&NiagaraSpriteRendererProperties]
  renderers [
   {rendererClass=&NiagaraSpriteRendererProperties}
   {rendererClass=&NiagaraSpriteRendererProperties, rendererIndex=1}
   {rendererClass=&NiagaraSpriteRendererProperties, rendererIndex=2}
  ]
 }
 Stars_No_Trail001 {
  emitterUpdateScript [
   &EmitterState(
    Inactive Response=::NewEnumerator1
    Life Cycle Mode=::NewEnumerator1
    Loop Behavior=::NewEnumerator1
    Loop Duration Mode=::NewEnumerator1
    UseLoopDelay=false
   )
   &SpawnBurst_Instantaneous(Spawn Count=2, Use Spawn Probability=false, UseLoopCountLimit=false)
   &SpawnBurst_Instantaneous(Spawn Group=1, Use Spawn Probability=false, UseLoopCountLimit=false) {enabled=false, module=SpawnBurst_Instantaneous001}
  ]
  particleSpawnScript [
   &FN_SetSpawnGroup(Number of Spawn Groups=2) {module=FN_SetSpawnGroup001}
   &InitializeParticle(
    Color=&MakeLinearColorFromFloat(Value:NiagaraFloat=&FN_FrontEndFloatModifier(Gameplay=1, Front End=2))
    Mesh Scale=<0.2 0.2 0.4>
    Mesh Scale Mode=::NewEnumerator4
    UsePositionOffset=false
    Write Material Random=true
   )
   &SubUVAnimation(End Frame=1, SubUV Animation Mode:ENiagara_SubUVLookupMode=::NewEnumerator2)
   /SetVariables(Particles.VisibilityTag=&MakeIntFromFloat(Float=$User.Variant_Float)) {isSetParameters=true, module=SetVariables_#10}
   &CameraOffset()
  ]
  particleUpdateScript [
   &ParticleState(Kill Particles When Lifetime Has Elapsed=false)
   &SkeletalMeshLocation(
    Bone / Socket Index=$Particles.SpawnGroup
    Bone Sampling Mode=::NewEnumerator4
    Inherit Velocity=false
    MaskSpawnGroup=false
    Sampled Position Offset=&FN_FrontEndVector3Modifier(
     Gameplay=&Lerp_Vector(A=<0 -4 10>, B=<0 0 8.5>, Alpha=$System.RDive)
     Front End=&Lerp_Vector(
      A=<0 -4 5>
      B=&MaskVectorBySpawnGroup(Masked Vector=<3 -3 3>, Passthrough Vector=<-4.5 -3 5>)
      Alpha=$System.Dive
     )
    )
    Sampled Position Offset Coordinate Space:ENiagaraMeshTransforms=::NewEnumerator2
    Skeletal Mesh=$User.SkeletalMesh
   )
   &ScaleMeshSize(Scale Factor=&Lerp_Vector(A=<1 1 1>, B=<1 1 1.25>, Alpha=$System.RDive))
   &SolveForcesAndVelocity(Clamp Velocity=false, Limit Acceleration=false) {enabled=false}
   &UpdateMeshOrientation(
    Facing Interpolation=false
    Orientation Method=::NewEnumerator1
    Override Quat=false
    TransformFacing=true
    TransformUp=false
    Up Direction:Vector3f=&FN_FrontEndVector3Modifier(Gameplay=<0 -0.05 1>, Front End=&Lerp_Vector(A=<0 -0.05 1>, B=<0 -0.8 1>, Alpha=$System.Dive))
   )
  ]
  rendererClasses [&NiagaraMeshRendererProperties]
  renderers [
   {rendererClass=&NiagaraMeshRendererProperties, rendererIndex=0}
   {rendererClass=&NiagaraMeshRendererProperties, rendererIndex=1}
   {rendererClass=&NiagaraMeshRendererProperties, rendererIndex=2}
   {rendererClass=&NiagaraMeshRendererProperties, rendererIndex=3}
   {rendererClass=&NiagaraMeshRendererProperties, rendererIndex=4}
   {rendererClass=&NiagaraMeshRendererProperties, rendererIndex=5}
   {rendererClass=&NiagaraMeshRendererProperties, rendererIndex=6}
   {rendererClass=&NiagaraMeshRendererProperties, rendererIndex=7}
   {rendererClass=&NiagaraMeshRendererProperties, rendererIndex=8}
  ]
 }
 bracelets_L {
  emitterUpdateScript [
   &EmitterState(
    Inactive Response=::NewEnumerator1
    Life Cycle Mode=::NewEnumerator1
    Loop Behavior=::NewEnumerator1
    Loop Duration Mode=::NewEnumerator1
    UseLoopDelay=false
   )
   &SpawnBurst_Instantaneous(Use Spawn Probability=false, UseLoopCountLimit=false)
  ]
  particleSpawnScript [
   &InitializeParticle(
    Color=&MakeLinearColorFromFloat(Value:NiagaraFloat=&FN_FrontEndFloatModifier(Gameplay=1, Front End=2))
    Mesh Scale=<1.3 1.3 1.3>
    Mesh Scale Mode=::NewEnumerator4
    UsePositionOffset=false
    Write Material Random=true
   )
   &SubUVAnimation(End Frame=1, SubUV Animation Mode:ENiagara_SubUVLookupMode=::NewEnumerator2)
   /SetVariables(Particles.VisibilityTag=&MakeIntFromFloat(Float=$User.Variant_Float)) {isSetParameters=true, module=SetVariables_#10}
   &InitialMeshOrientation(Mesh Orientation Mode=::NewEnumerator2, Use Rotation Vector=false)
   &CameraOffset()
  ]
  particleUpdateScript [
   &ParticleState(Kill Particles When Lifetime Has Elapsed=false)
   &SkeletalMeshLocation(
    Bone Sampling Mode=::NewEnumerator4
    Inherit Velocity=false
    Interpolate Mesh Sampling=false
    MaskSpawnGroup=false
    Sampled Position Offset=&FN_FrontEndVector3Modifier(
     Gameplay=&Lerp_Vector(A=<3 7 2>, B=<1 8 7>, Alpha=$System.Dive)
     Front End=&Lerp_Vector(A=<0 9 -3>, B=<5 3 -3>, Alpha=$System.RDive)
    )
    Sampled Position Offset Coordinate Space:ENiagaraMeshTransforms=::NewEnumerator2
    Skeletal Mesh=$User.SkeletalMesh
   )
   &ScaleMeshSize(Scale Factor=&Lerp_Vector(A=<1 1 1>, B=<1 1 1>, Alpha=$System.RDive))
   &UpdateMeshOrientation(
    Facing Interpolation=false
    Orientation Method=::NewEnumerator1
    Override Quat=false
    TransformFacing=true
    TransformUp=false
    Up Direction:Vector3f=&FN_FrontEndVector3Modifier(Gameplay=<0 -0.05 1>, Front End=&Lerp_Vector(A=<0 -0.05 1>, B=<0 -0.8 1>, Alpha=$System.Dive))
   ) {enabled=false}
   /SetVariables(Particles.MeshOrientation=&LerpQuaternion(
     Quaternion A:Quat4f=<0 0 0 1>
     Quaternion B:Quat4f=<0 0 2 1>
     Lerp Factor:NiagaraFloat=$System.StepDive
     Enable Spherical Linear Interpolation:NiagaraBool=true
    )) {enabled=false, isSetParameters=true, module=SetVariables_#11}
  ]
  rendererClasses [&NiagaraMeshRendererProperties]
  renderers [
   {rendererClass=&NiagaraMeshRendererProperties, rendererIndex=0}
   {rendererClass=&NiagaraMeshRendererProperties, rendererIndex=1}
   {rendererClass=&NiagaraMeshRendererProperties, rendererIndex=2}
   {rendererClass=&NiagaraMeshRendererProperties, rendererIndex=3}
   {rendererClass=&NiagaraMeshRendererProperties, rendererIndex=4}
   {rendererClass=&NiagaraMeshRendererProperties, rendererIndex=5}
  ]
 }
 bracelets_R {
  emitterUpdateScript [
   &EmitterState(
    Inactive Response=::NewEnumerator1
    Life Cycle Mode=::NewEnumerator1
    Loop Behavior=::NewEnumerator1
    Loop Duration Mode=::NewEnumerator1
    UseLoopDelay=false
   )
   &SpawnBurst_Instantaneous(Use Spawn Probability=false, UseLoopCountLimit=false)
   &SpawnBurst_Instantaneous(Spawn Time=2.3254, Use Spawn Probability=false, UseLoopCountLimit=false) {module=SpawnBurst_Instantaneous001}
  ]
  particleSpawnScript [
   &InitializeParticle(
    Color=&MakeLinearColorFromFloat(Value:NiagaraFloat=&FN_FrontEndFloatModifier(Gameplay=1, Front End=2))
    Mesh Scale=<1.3 1.3 1.3>
    Mesh Scale Mode=::NewEnumerator4
    UsePositionOffset=false
    Write Material Random=true
   )
   &SubUVAnimation(End Frame=1, SubUV Animation Mode:ENiagara_SubUVLookupMode=::NewEnumerator2)
   /SetVariables(Particles.VisibilityTag=&MakeIntFromFloat(Float=$User.Variant_Float)) {isSetParameters=true, module=SetVariables_#10}
   &InitialMeshOrientation(Mesh Orientation Mode=::NewEnumerator2, Use Rotation Vector=false)
   &CameraOffset()
  ]
  particleUpdateScript [
   &ParticleState(Kill Particles When Lifetime Has Elapsed=false)
   &SkeletalMeshLocation(
    Bone / Socket Index=1
    Bone Sampling Mode=::NewEnumerator4
    Inherit Velocity=false
    Interpolate Mesh Sampling=false
    MaskSpawnGroup=false
    Sampled Position Offset=&FN_FrontEndVector3Modifier(
     Gameplay=&Lerp_Vector(A=<0 7 2>, B=<0 8 7>, Alpha=$System.Dive)
     Front End=&Lerp_Vector(A=<0 9 -3>, B=<-5 3 -3>, Alpha=$System.RDive)
    )
    Sampled Position Offset Coordinate Space:ENiagaraMeshTransforms=::NewEnumerator2
    Skeletal Mesh=$User.SkeletalMesh
   )
   &ScaleMeshSize(Scale Factor=&Lerp_Vector(A=<1 1 1>, B=<1 1 1>, Alpha=$System.RDive))
   &UpdateMeshOrientation(
    Facing Interpolation=false
    Orientation Method=::NewEnumerator1
    Override Quat=false
    TransformFacing=true
    TransformUp=false
    Up Direction:Vector3f=&FN_FrontEndVector3Modifier(Gameplay=<0 -0.05 1>, Front End=&Lerp_Vector(A=<0 -0.05 1>, B=<0 -0.8 1>, Alpha=$System.Dive))
   ) {enabled=false}
   /SetVariables(Particles.MeshOrientation=&LerpQuaternion(
     Quaternion A:Quat4f=<0 0 0 1>
     Quaternion B:Quat4f=<0 0 2 1>
     Lerp Factor:NiagaraFloat=$System.StepDive
     Enable Spherical Linear Interpolation:NiagaraBool=true
    )) {enabled=false, isSetParameters=true, module=SetVariables_#11}
  ]
  rendererClasses [&NiagaraMeshRendererProperties]
  renderers [
   {rendererClass=&NiagaraMeshRendererProperties, rendererIndex=0}
   {rendererClass=&NiagaraMeshRendererProperties, rendererIndex=1}
   {rendererClass=&NiagaraMeshRendererProperties, rendererIndex=2}
   {rendererClass=&NiagaraMeshRendererProperties, rendererIndex=3}
   {rendererClass=&NiagaraMeshRendererProperties, rendererIndex=4}
   {rendererClass=&NiagaraMeshRendererProperties, rendererIndex=5}
  ]
 }
}
summary {
 systemName=NS_Contrail_PinkSpike
 emitters [
  {emitterName=Stars, rendererClasses [&NiagaraSpriteRendererProperties]}
  {emitterName=Ribbon_Follow, rendererClasses [&NiagaraRibbonRendererProperties]}
  {emitterName=Speedlines, simTarget=GPUComputeSim, rendererClasses [&NiagaraSpriteRendererProperties]}
  {emitterName=Stars_No_Trail, simTarget=GPUComputeSim, rendererClasses [&NiagaraSpriteRendererProperties]}
  {emitterName=Smoke_Large, rendererClasses [&NiagaraMeshRendererProperties]}
  {emitterName=Smoke_Small, rendererClasses [&NiagaraMeshRendererProperties]}
  {emitterName=Stars_No_Trail001, rendererClasses [&NiagaraMeshRendererProperties]}
  {emitterName=bracelets_L, rendererClasses [&NiagaraMeshRendererProperties]}
  {emitterName=bracelets_R, rendererClasses [&NiagaraMeshRendererProperties]}
 ]
 userVariables [
  User.bIsFrontEnd {type=NiagaraBool}
  User.bIsFrontEndPreview {type=NiagaraBool}
  User.IsSkydiving {type=NiagaraBool}
  User.ParaGlideLeanAlpha {type=NiagaraFloat, defaultValue=0, description=""}
  User.SkeletalMesh {
   type=NiagaraDataInterfaceSkeletalMesh
   defaultValue {
    struct=NiagaraExt_VariableValue_DataInterface
    value {dataInterface=&self:NiagaraDataInterfaceSkeletalMesh_0}
   }
  }
  User.Variant_Float {type=NiagaraFloat, defaultValue=0, description=""}
 ]
}
